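; Parent gameplay ability for HighTower items that clears a cooldown when the owning pawn lands.
; On activation it binds to the FortPawn's OnPawnLanded event and starts tag-add waits for surface swimming and being in a vehicle.
; Landing or gaining either tag routes to CallRemove, which removes gameplay effects from the owner by a configurable asset tag.
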(event Custom|SetUpBinds
  ;[1 1 1] Set Up Binds
  (bind _asfort_pawn (Utilities|Casting|CastToFortPawn (Ability|GetAvatarActorfromActorInfo)))
  (EventDispatchers|BindEventtoOnPawnLanded _asfort_pawn (delegate Custom|PawnLanded)))

(event Custom|PawnLanded
  (params (Hit :type Hit%20Result%20Structure%20%28by%20ref%29))
  ;[1 1 1] Set Up Binds
  (CallFunction|CallRemove))

(event Custom|SetUpWaits
  ;[1 1 1] Set Up Waits
  (Ability|Tasks|WaitGameplayTagAdd (Variables|Default|GetTag_SurfaceSwimming)
    (:then
      (Ability|Tasks|WaitGameplayTagAdd (Variables|Default|GetTag_InVehicle)
        (:Added
          _)))
    (:Added
      _))
  (CallFunction|CallRemove))

(event Ability|EventActivateAbility
  ;[1 1 1] Set up the waits and binds that will remove the cooldown
  (CallFunction|SetUpWaits)
  (CallFunction|SetUpBinds))

(event Custom|CallRemove
  ;[1 1 1] Call Remove - Remove the cooldowns
  (Ability|RemoveGameplayEffectFromOwnerWithAssetTags (Variables|Default|GetTag_AssetTagToRemove)))
